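# Anim Blueprint for the Mind Pinch NPC body cosmetic, deriving its pose from the attached parent mesh.
# Copy Pose From Mesh feeds a Control Rig, then head_offset gets HeadRot as a bone-space replacement rotation.
# Event Graph builds HeadRot from the owner's head-to-spine_05 delta, scaled per axis, with stronger yaw when crouching.
# A Copy Bone Delta chain transfers rotation only from thumb and index bones to their fake counterparts on both hands.
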AnimBlueprint /BRCosmetics/Characters/Player/Male/Medium/Bodies/M_MED_MindPinch_NPC/M_MED_Mind_Pinch_NPC_AnimBP
target_skeleton=&/M_MED_Mind_Pinch_NPC_Skeleton
graphs [
 AnimGraph {nodes [
   n1 = Control Rig(Source=Copy Pose From Mesh() {bCopyCurves=true, bUseAttachedParent=true}) {
    alphaBoolBlend {blendOption=Linear}
    controlRigAssetReference {blueprintRigClass=&/M_MED_MindPinch_NPC_CtrlRig.M_MED_MindPinch_NPC_CtrlRig_C}
   }
   n2 = "Transform (Modify) Bone - Bone: head_offset"(ComponentPose=Local To Component(LocalPose=@n1), Alpha=1, Translation="0,0,0", Rotation=Get HeadRot() -> (HeadRot:Rotator), Scale="1,1,1") {rotationMode=BMM_Replace, rotationSpace=BCS_BoneSpace, alphaBoolBlend {blendOption=Linear}, boneToModify {boneName=head_offset}}
   n3 = Copy Bone Delta - Source Bone: thumb_02_r - Target Bone: thumb_fake_02_r!(
    ComponentPose:Component Space Pose Link Structure=Copy Bone Delta - Source Bone: thumb_03_l - Target Bone: thumb_fake_03_l!(
     ComponentPose:Component Space Pose Link Structure=Copy Bone Delta - Source Bone: thumb_02_l - Target Bone: thumb_fake_02_l!(
      ComponentPose:Component Space Pose Link Structure=@n2
      Alpha:float=1
      bCopyTranslation:bool=False
      bCopyRotation:bool=true
      bCopyScale:bool=False
     ) -> (Pose:Component Space Pose Link Structure) {
      is=title AnimGraphNode_CopyBoneDelta
      alphaBoolBlend {blendOption=Linear}
      sourceBone {boneName=thumb_02_l}
      targetBone {boneName=thumb_fake_02_l}
     }
     Alpha:float=1
     bCopyTranslation:bool=False
     bCopyRotation:bool=true
     bCopyScale:bool=False
    ) -> (Pose:Component Space Pose Link Structure) {
     is=title AnimGraphNode_CopyBoneDelta
     alphaBoolBlend {blendOption=Linear}
     sourceBone {boneName=thumb_03_l}
     targetBone {boneName=thumb_fake_03_l}
    }
    Alpha:float=1
    bCopyTranslation:bool=False
    bCopyRotation:bool=true
    bCopyScale:bool=False
   ) -> (Pose:Component Space Pose Link Structure) {
    is=title AnimGraphNode_CopyBoneDelta
    alphaBoolBlend {blendOption=Linear}
    sourceBone {boneName=thumb_02_r}
    targetBone {boneName=thumb_fake_02_r}
   }
   n4 = Copy Bone Delta - Source Bone: index_02_l - Target Bone: index_fake_02_l!(
    ComponentPose:Component Space Pose Link Structure=Copy Bone Delta - Source Bone: index_01_l - Target Bone: index_fake_01_l!(
     ComponentPose:Component Space Pose Link Structure=Copy Bone Delta - Source Bone: thumb_03_r - Target Bone: thumb_fake_03_r!(
      ComponentPose:Component Space Pose Link Structure=@n3
      Alpha:float=1
      bCopyTranslation:bool=False
      bCopyRotation:bool=true
      bCopyScale:bool=False
     ) -> (Pose:Component Space Pose Link Structure) {
      is=title AnimGraphNode_CopyBoneDelta
      alphaBoolBlend {blendOption=Linear}
      sourceBone {boneName=thumb_03_r}
      targetBone {boneName=thumb_fake_03_r}
     }
     Alpha:float=1
     bCopyTranslation:bool=False
     bCopyRotation:bool=true
     bCopyScale:bool=False
    ) -> (Pose:Component Space Pose Link Structure) {
     is=title AnimGraphNode_CopyBoneDelta
     alphaBoolBlend {blendOption=Linear}
     sourceBone {boneName=index_01_l}
     targetBone {boneName=index_fake_01_l}
    }
    Alpha:float=1
    bCopyTranslation:bool=False
    bCopyRotation:bool=true
    bCopyScale:bool=False
   ) -> (Pose:Component Space Pose Link Structure) {
    is=title AnimGraphNode_CopyBoneDelta
    alphaBoolBlend {blendOption=Linear}
    sourceBone {boneName=index_02_l}
    targetBone {boneName=index_fake_02_l}
   }
   n5 = Copy Bone Delta - Source Bone: index_02_r - Target Bone: index_fake_02_r!(
    ComponentPose:Component Space Pose Link Structure=Copy Bone Delta - Source Bone: index_01_r - Target Bone: index_fake_01_r!(
     ComponentPose:Component Space Pose Link Structure=Copy Bone Delta - Source Bone: index_03_l - Target Bone: index_fake_03_l!(
      ComponentPose:Component Space Pose Link Structure=@n4
      Alpha:float=1
      bCopyTranslation:bool=False
      bCopyRotation:bool=true
      bCopyScale:bool=False
     ) -> (Pose:Component Space Pose Link Structure) {
      is=title AnimGraphNode_CopyBoneDelta
      alphaBoolBlend {blendOption=Linear}
      sourceBone {boneName=index_03_l}
      targetBone {boneName=index_fake_03_l}
     }
     Alpha:float=1
     bCopyTranslation:bool=False
     bCopyRotation:bool=true
     bCopyScale:bool=False
    ) -> (Pose:Component Space Pose Link Structure) {
     is=title AnimGraphNode_CopyBoneDelta
     alphaBoolBlend {blendOption=Linear}
     sourceBone {boneName=index_01_r}
     targetBone {boneName=index_fake_01_r}
    }
    Alpha:float=1
    bCopyTranslation:bool=False
    bCopyRotation:bool=true
    bCopyScale:bool=False
   ) -> (Pose:Component Space Pose Link Structure) {
    is=title AnimGraphNode_CopyBoneDelta
    alphaBoolBlend {blendOption=Linear}
    sourceBone {boneName=index_02_r}
    targetBone {boneName=index_fake_02_r}
   }
   Output Pose(Result=Component To Local(ComponentPose=Copy Bone Delta - Source Bone: index_03_r - Target Bone: index_fake_03_r!(
      ComponentPose:Component Space Pose Link Structure=@n5
      Alpha:float=1
      bCopyTranslation:bool=False
      bCopyRotation:bool=true
      bCopyScale:bool=False
     ) -> (Pose:Component Space Pose Link Structure) {
      is=title AnimGraphNode_CopyBoneDelta
      alphaBoolBlend {blendOption=Linear}
      sourceBone {boneName=index_03_r}
      targetBone {boneName=index_fake_03_r}
     })) {layerGroup=None}
  ]}
 EventGraph {nodes [
   Event BlueprintUpdateAnimation()
   > n1 = Sequence()
   > n2 = Branch(Condition=Get bIsCrouching())
   > Set HeadRotYawPCT(HeadRotYawPCT:double=0.8) -> (Output_Get:double)
   Set HeadRotYawPCT(execute=@n2.else, HeadRotYawPCT=0.6)
   n3 = GetOwnerSkeletalMeshComponent()
   n4 = BreakRotator(InRot=BreakTransform(InTransform=GetDeltaTransformFromRefPose(self=@n3, BoneName=head, BaseName=spine_05)).Rotation)
   n5 = Get HeadRotYawPCT() -> (HeadRotYawPCT:double)
   Set HeadRot(
    execute=@n1.then_1
    HeadRot:Rotator=MakeRotator(
     Roll=float * float(A=@n4, B=0.75) {is=title K2Node_PromotableOperator}
     Pitch=float * float(A=@n4.Pitch, B=0.5)
     Yaw=float * float(A=@n4.Yaw, B=@n5)
    )
   ) -> (Output_Get:Rotator)
  ]}
]
variables [
 HeadRot {type=Rotator}
 HeadRotYawPCT {type=number}
]
